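# Editor utility scripted action that exports selected assets, intended for Chaos cloth assets, to a chosen project Content directory.
# For each selection, embedded Python gathers Material and PhysicsAsset dependencies from the Asset Registry.
# It also adds non-script dependencies of the first one found, assumed to be the material.
# It passes these paths and their parent folders to the migration tool, prompting and ignoring automatic dependencies.
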EditorUtilityBlueprint /ChaosClothAsset/BP_Cloth_Asset_Export
graphs [
 EventGraph {links [], nodes []}
 export_cloth_asset {nodes [
   n1 = export_cloth_asset!() -> (then, Destionation Project Content Dir:Directory Path Structure) {is=title K2Node_FunctionEntry}
   > n2 = GetSelectedAssets()
   > n3 = For Each Loop(Array:Array of Object References=@n2) -> (Array Element:Object Reference)
   > Execute Python Script!(
    execute
    PythonScript="import unreal\r\n\r\n##input from the scripted asset action\r\nselected_asset_path = selected_asset_path.get_outer().get_path_name()\r\n\r\nasset_registry_helpers = unreal.AssetRegistryHelpers.get_asset_registry()\r\nasset = asset_registry_helpers.get_asset_by_object_path(selected_asset_path)\r\n\r\nasset_paths = []\r\n##Add the selected item into our collection    \r\nasset_paths.append(selected_asset_path)\r\n\r\n##get top level dependencies\r\ndependency_options = unreal.AssetRegistryDependencyOptions()\r\ndependencies = asset_registry_helpers.get_dependencies(asset.package_name, dependency_options)\r\n\r\n##Find hand picked assets\r\nfor dependency in dependencies:\r\n    asset_data = unreal.EditorAssetLibrary().find_asset_data(dependency)\r\n    if asset_data.asset_class_path.asset_name == \"Material\":\r\n        asset_paths.append(str(dependency))\r\n    elif asset_data.asset_class_path.asset_name == \"PhysicsAsset\":\r\n        asset_paths.append(str(dependency))\r\n\r\ntry:\r\n    ##Get Material Dependencies \r\n    asset = asset_registry_helpers.get_asset_by_object_path(asset_paths[1])\r\n    dependencies = asset_registry_helpers.get_dependencies(asset.package_name, dependency_options)\r\n\r\n    asset_paths.extend([str(dependency) for dependency in dependencies if not str(dependency).startswith(\"/Script\")])\r\nexcept IndexError as e:\r\n    print(f\"index doesn't exist \\n{e}\")\r\n\r\n\r\n##Add top folders to highlight\r\ngame_asset_paths = []\r\nfor asset_path in asset_paths:\r\n    current_path = ''\r\n    for folder in asset_path.split('/')[:-1]:\r\n        if folder:\r\n            current_path += '/' + folder \r\n            game_asset_paths.append(current_path)\r\n\r\ngame_asset_paths.extend(asset_paths)\r\ngame_asset_paths = set(game_asset_paths)\r\n\r\n\r\n##Create Migration Tool assets\r\nhelpers = unreal.AssetToolsHelpers()\r\nmigration_options = unreal.MigrationOptions()\r\nmigration_options.orphan_folder\r\nmigration_options.set_editor_property(\"prompt\", True)\r\nmigration_options.set_editor_property(\"ignore_dependencies\", True)\r\n\r\n##Launch migration Tool\r\nasset_tools = helpers.get_asset_tools()\r\nasset_tools.migrate_packages(game_asset_paths, destination_project_content_dir, migration_options)"
    selected_asset_path:Object Reference=@n3
    destination_project_content_dir:String=Break Directory Path!(DirectoryPath:"Directory Path Structure (by ref)"=@n1) -> (Path:String) {is=title K2Node_BreakStruct}
   ) -> (then, ReturnValue) {inputs [selected_asset_path, destination_project_content_dir]}
  ]}
]
variables []
