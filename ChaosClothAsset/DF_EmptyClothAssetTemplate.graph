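# Empty Chaos Cloth Dataflow template, a blank starting graph for cloth assets.
# It holds a single ClothAssetTerminal node, version 2, with no connections, so no import, simulation or render steps feed the terminal.
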Dataflow /ChaosClothAsset/DF_EmptyClothAssetTemplate
name=DF_EmptyClothAssetTemplate
connections []
nodes [ClothAssetTerminal {type=FChaosClothAssetTerminalNode_v2, guid=#1}]
