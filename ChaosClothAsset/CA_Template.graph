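# Template Chaos Cloth asset, backed by a Dataflow graph named DF_EmptyClothAssetTemplate.
# The graph holds a single cloth asset terminal node with no connections, so it provides an empty starting point with nothing feeding the output.
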ChaosClothAsset /ChaosClothAsset/CA_Template
name=DF_EmptyClothAssetTemplate
graph=&/DF_EmptyClothAssetTemplate
connections []
nodes [ClothAssetTerminal {type=FChaosClothAssetTerminalNode_v2, guid=#1}]
