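# DF_ClothSolver is a Chaos Cloth Dataflow simulation graph that steps the physics solvers of the Cloth simulation group.
# GetPhysicsSolvers collects the Cloth-group solvers and GetSimulationTime supplies the simulation time, both feeding AdvancePhysicsSolvers.
# The advanced solvers then pass to SimulationProxiesTerminal as the simulation proxies output, with no other processing in between.
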Dataflow /ChaosClothAsset/DF_ClothSolver
name=DF_ClothSolver
connections [
 {
  fromNode=GetPhysicsSolvers
  fromPin=PhysicsSolvers
  fromPinGuid=#1
  toNode=AdvancePhysicsSolvers
  toPin=PhysicsSolvers
  toPinGuid=#2
 }
 {
  fromNode=GetSimulationTime
  fromPin=SimulationTime
  fromPinGuid=#3
  toNode=AdvancePhysicsSolvers
  toPin=SimulationTime
  toPinGuid=#4
 }
 {
  fromNode=AdvancePhysicsSolvers
  fromPin=PhysicsSolvers
  fromPinGuid=#5
  toNode=SimulationProxiesTerminal
  toPin=SimulationProxies
  toPinGuid=#6
 }
]
nodes [
 GetPhysicsSolvers {
  type=FGetPhysicsSolversDataflowNode
  guid=#7
  inputPins []
  outputPins [PhysicsSolvers {type="TArray<FDataflowSimulationProperty>"}]
  properties {SimulationGroups=("Cloth")}
 }
 AdvancePhysicsSolvers {
  type=FAdvancePhysicsSolversDataflowNode
  guid=#8
  inputPins [
   SimulationTime {type=FDataflowSimulationTime}
   PhysicsSolvers {type="TArray<FDataflowSimulationProperty>"}
  ]
  outputPins [PhysicsSolvers {type="TArray<FDataflowSimulationProperty>"}]
 }
 GetSimulationTime {
  type=FGetSimulationTimeDataflowNode
  guid=#9
  inputPins []
  outputPins [SimulationTime {type=FDataflowSimulationTime}]
 }
 SimulationProxiesTerminal {
  type=FSimulationProxiesTerminalDataflowNode
  guid=#10
  inputPins [SimulationProxies {type="TArray<FDataflowSimulationProperty>"}]
  outputPins []
 }
]
