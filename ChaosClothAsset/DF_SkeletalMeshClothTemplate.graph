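# Template Chaos Cloth Dataflow graph that builds a cloth asset from skeletal meshes, with empty mesh and physics asset slots.
# Separate simulation and render mesh imports merge, then pass through a MaxDistance weight map.
# The max distance config supplies kinematic vertices to geodesic long-range tethers.
# Stretch, bending, mass, aerodynamics, damping, self-collision spheres, physics asset and collision follow, then unused attributes are stripped before solver setup and the LOD0 terminal.
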Dataflow /ChaosClothAsset/DF_SkeletalMeshClothTemplate
name=DF_SkeletalMeshClothTemplate
connections [
 {
  fromNode=MergeClothCollections
  fromPin=Collection
  fromPinGuid=#1
  toNode=WeightMap_MaxDistance
  toPin=Collection
  toPinGuid=#2
 }
 {
  fromNode=SimulationSelfCollisionSpheresConfig
  fromPin=Collection
  fromPinGuid=#3
  toNode=SetPhysicsAsset
  toPin=Collection
  toPinGuid=#4
 }
 {
  fromNode=SimulationLongRangeAttachmentConfig
  fromPin=Collection
  fromPinGuid=#5
  toNode=SimulationStretchConfig
  toPin=Collection
  toPinGuid=#6
 }
 {
  fromNode=SimulationStretchConfig
  fromPin=Collection
  fromPinGuid=#7
  toNode=SimulationBendingConfig
  toPin=Collection
  toPinGuid=#8
 }
 {
  fromNode=WeightMap_MaxDistance
  fromPin=Collection
  fromPinGuid=#9
  toNode=SimulationMaxDistanceConfig
  toPin=Collection
  toPinGuid=#10
 }
 {
  fromNode=WeightMap_MaxDistance
  fromPin=OutputName.StringValue
  fromPinGuid=#11
  toNode=SimulationMaxDistanceConfig
  toPin=MaxDistance.WeightMap
  toPinGuid=#12
 }
 {
  fromNode=SetPhysicsAsset
  fromPin=Collection
  fromPinGuid=#13
  toNode=SimulationCollisionConfig
  toPin=Collection
  toPinGuid=#14
 }
 {
  fromNode=SimulationBendingConfig
  fromPin=Collection
  fromPinGuid=#15
  toNode=SimulationMassConfig
  toPin=Collection
  toPinGuid=#16
 }
 {
  fromNode=SimulationMassConfig
  fromPin=Collection
  fromPinGuid=#17
  toNode=SimulationAerodynamicsConfig
  toPin=Collection
  toPinGuid=#18
 }
 {
  fromNode=SimulationAerodynamicsConfig
  fromPin=Collection
  fromPinGuid=#19
  toNode=SimulationDampingConfig
  toPin=Collection
  toPinGuid=#20
 }
 {
  fromNode=SimulationDampingConfig
  fromPin=Collection
  fromPinGuid=#21
  toNode=SimulationSelfCollisionSpheresConfig
  toPin=Collection
  toPinGuid=#22
 }
 {
  fromNode=SimulationSolverConfig LOD0
  fromPin=Collection
  fromPinGuid=#23
  toNode=ClothAssetTerminal
  toPin="CollectionLods[0]"
  toPinGuid=#24
 }
 {
  fromNode=SimulationCollisionConfig
  fromPin=Collection
  fromPinGuid=#25
  toNode=StripUserAttributes
  toPin=Collection
  toPinGuid=#26
 }
 {
  fromNode=SimulationMaxDistanceConfig
  fromPin=Collection
  fromPinGuid=#27
  toNode=SimulationLongRangeAttachmentConfig
  toPin=Collection
  toPinGuid=#28
 }
 {
  fromNode=SimulationMaxDistanceConfig
  fromPin=KinematicVertices3D
  fromPinGuid=#29
  toNode=SimulationLongRangeAttachmentConfig
  toPin=FixedEndSet.StringValue
  toPinGuid=#30
 }
 {
  fromNode=SkeletalMeshImport_RENDER
  fromPin=Collection
  fromPinGuid=#31
  toNode=MergeClothCollections
  toPin="Collections[1]"
  toPinGuid=#32
 }
 {
  fromNode=SkeletalMeshImport_SIM
  fromPin=Collection
  fromPinGuid=#33
  toNode=MergeClothCollections
  toPin="Collections[0]"
  toPinGuid=#34
 }
 {
  fromNode=SkeletalMesh_SIM
  fromPin=SkeletalMesh
  fromPinGuid=#35
  toNode=SkeletalMeshImport_SIM
  toPin=SkeletalMesh
  toPinGuid=#36
 }
 {
  fromNode=SkeletalMesh_RENDER
  fromPin=SkeletalMesh
  fromPinGuid=#37
  toNode=SkeletalMeshImport_RENDER
  toPin=SkeletalMesh
  toPinGuid=#38
 }
 {
  fromNode=StripUserAttributes
  fromPin=Collection
  fromPinGuid=#39
  toNode=SimulationSolverConfig LOD0
  toPin=Collection
  toPinGuid=#40
 }
]
nodes [
 SetPhysicsAsset {type=FChaosClothAssetSetPhysicsAssetNode, guid=#41, properties {PhysicsAsset=None}}
 SimulationStretchConfig {
  type=FChaosClothAssetSimulationStretchConfigNode
  guid=#42
  properties {
   StretchAnisoDamping=(bIsAnimatable=True,Low=0.0001,High=0.0001,WeightMap="StretchAnisoDamping")
   StretchStiffnessBias=(bIsAnimatable=True,Low=12.9211,High=502.4609,WeightMap="StretchStiffnessBias")
   StretchStiffnessWarp=(bIsAnimatable=True,Low=159.1384,High=490.6552,WeightMap="StretchStiffnessWarp")
   StretchStiffnessWeft=(bIsAnimatable=True,Low=254.5566,High=629.4298,WeightMap="StretchStiffnessWeft")
  }
 }
 SimulationBendingConfig {
  type=FChaosClothAssetSimulationBendingConfigNode
  guid=#43
  inputPins [
   Collection {type=FManagedArrayCollection}
   FlatnessRatio.WeightMap {type=FString}
   RestAngle.WeightMap {type=FString}
   BendingStiffness.WeightMap {type=FString}
   BendingStiffnessWarp.WeightMap {type=FString}
   BendingStiffnessWeft.WeightMap {type=FString}
   BendingStiffnessBias.WeightMap {type=FString}
   BendingDamping.WeightMap {type=FString}
   BendingAnisoDamping.WeightMap {type=FString}
   AnisoBucklingRatio.WeightMap {type=FString}
   BucklingStiffness.WeightMap {type=FString}
   BucklingStiffnessWarp.WeightMap {type=FString}
   BucklingStiffnessWeft.WeightMap {type=FString}
   BucklingStiffnessBias.WeightMap {type=FString}
   BucklingRatioWeighted.WeightMap {type=FString}
  ]
  properties {
   AnisoBucklingRatio=(bIsAnimatable=True,Low=0.3,High=0.3,WeightMap="BucklingRatio")
   BendingAnisoDamping=(bIsAnimatable=True,Low=0.0001,High=0.0001,WeightMap="BendingAnisoDamping")
   BendingStiffnessBias=(bIsAnimatable=True,Low=0.025,High=0.3068,WeightMap="BendingStiffnessBias")
   BendingStiffnessWarp=(bIsAnimatable=True,Low=0.0305,High=13,WeightMap="BendingStiffnessWarp")
   BendingStiffnessWeft=(bIsAnimatable=True,Low=0.0196,High=0.1843,WeightMap="BendingStiffnessWeft")
   BucklingStiffnessBias=(bIsAnimatable=True,Low=0.0062,High=0.0767,WeightMap="BucklingStiffnessBias")
   BucklingStiffnessWarp=(bIsAnimatable=True,Low=0.0076,High=3.25,WeightMap="BucklingStiffnessWarp")
   BucklingStiffnessWeft=(bIsAnimatable=True,Low=0.0049,High=0.0461,WeightMap="BucklingStiffnessWeft")
  }
 }
 SimulationMaxDistanceConfig {
  type=FChaosClothAssetSimulationMaxDistanceConfigNode
  guid=#44
  properties {MaxDistance=(bIsAnimatable=True,High=100,WeightMap="MaxDistance")}
 }
 SimulationCollisionConfig {
  type=FChaosClothAssetSimulationCollisionConfigNode
  guid=#45
  properties {
   ClothCollisionThicknessImported=(bIsAnimatable=True,WeightMap="ClothCollisionThickness")
   CollisionThicknessImported=(ImportedValue=1)
   FrictionCoefficientWeighted=(bIsAnimatable=True,Low=0.8,High=0.8,WeightMap="FrictionCoefficient")
   InnerCollisionThickness=(WeightMap="InnerCollisionThickness")
   bUseCCD=False
  }
 }
 SimulationMassConfig {
  type=FChaosClothAssetSimulationMassConfigNode
  guid=#46
  properties {DensityWeighted=(Low=0.0813,High=0.2778,WeightMap="Density")}
 }
 SimulationAerodynamicsConfig {
  type=FChaosClothAssetSimulationAerodynamicsConfigNode
  guid=#47
  properties {
   Drag=(bIsAnimatable=True,Low=0.035,High=1,WeightMap="Drag")
   Lift=(bIsAnimatable=True,Low=0.035,High=1,WeightMap="Lift")
   TurbulenceRatio=1
   WindVelocity=(X=0,Y=0,Z=0)
   WindVelocitySpace=WorldSpace
  }
 }
 SimulationDampingConfig {
  type=FChaosClothAssetSimulationDampingConfigNode
  guid=#48
  properties {
   DampingCoefficientWeighted=(bIsAnimatable=True,Low=0.01,High=0.01,WeightMap="DampingCoefficient")
   LocalDampingAngularCoefficient=0
   LocalDampingLinearCoefficient=0
   LocalDampingSpace=CenterOfMass
  }
 }
 SimulationSelfCollisionSpheresConfig {
  type=FChaosClothAssetSimulationSelfCollisionSpheresConfigNode
  guid=#49
  inputPins [Collection {type=FManagedArrayCollection}]
  outputPins [
   Collection {type=FManagedArrayCollection}
   SelfCollisionSphereSetName {type=FString}
  ]
  properties {
   SelfCollisionSphereRadius=0.5
   SelfCollisionSphereRadiusCullMultiplier=1
   SelfCollisionSphereSetName=SelfCollisionSphereSetName
   SelfCollisionSphereStiffness=1
   bWarnDuplicateProperty=True
  }
 }
 ClothAssetTerminal {type=FChaosClothAssetTerminalNode_v2, guid=#50}
 WeightMap_MaxDistance {type=FChaosClothAssetWeightMapNode, guid=#51, properties {OutputName=(StringValue="MaxDistance")}}
 SimulationLongRangeAttachmentConfig {
  type=FChaosClothAssetSimulationLongRangeAttachmentConfigNode_v2
  guid=#52
  inputPins [
   Collection {type=FManagedArrayCollection}
   FixedEndSet.StringValue {type=FString}
   TetherStiffness.WeightMap {type=FString}
   TetherScale.WeightMap {type=FString}
   "CustomTetherData[0].CustomFixedEndSet.StringValue" {type=FString}
   "CustomTetherData[0].CustomDynamicEndSet.StringValue" {type=FString}
  ]
  outputPins [Collection {type=FManagedArrayCollection}]
  properties {
   CustomTetherData=(())
   FixedEndSet=(StringValue="KinematicVertices3D")
   TetherScale=(bIsAnimatable=True,Low=1,High=1,WeightMap="TetherScale")
   TetherStiffness=(bIsAnimatable=True,Low=1,High=1,WeightMap="TetherStiffness")
   bEnableCustomTetherGeneration=False
   bUseGeodesicTethers=True
   bWarnDuplicateProperty=True
  }
 }
 SimulationSolverConfig LOD0 {
  type=FChaosClothAssetSimulationSolverConfigNode
  guid=#53
  properties {MaxNumIterations=4, NumIterations=2, NumSubstepsImported=(ImportedValue=6)}
 }
 MergeClothCollections {type=FChaosClothAssetMergeClothCollectionsNode_v2, guid=#54}
 SkeletalMeshImport_SIM {
  type=FChaosClothAssetSkeletalMeshImportNode_v2
  guid=#55
  properties {SkeletalMesh=None, bImportRenderMesh=False, bImportSimMesh=True, bImportSingleSection=False}
 }
 SkeletalMeshImport_RENDER {
  type=FChaosClothAssetSkeletalMeshImportNode_v2
  guid=#56
  properties {SkeletalMesh=None, bImportRenderMesh=True, bImportSimMesh=False, bImportSingleSection=False}
 }
 SkeletalMesh_SIM {
  type=FGetSkeletalMeshDataflowNode
  guid=#57
  inputPins []
  outputPins [SkeletalMesh {type="TObjectPtr<USkeletalMesh>"}]
  properties {PropertyName=SkeletalMesh, SkeletalMesh=None}
 }
 SkeletalMesh_RENDER {
  type=FGetSkeletalMeshDataflowNode
  guid=#58
  inputPins []
  outputPins [SkeletalMesh {type="TObjectPtr<USkeletalMesh>"}]
  properties {PropertyName=SkeletalMesh, SkeletalMesh=None}
 }
 StripUserAttributes {
  type=FChaosClothStripUserAttributesNode
  guid=#59
  inputPins [
   Collection {type=FManagedArrayCollection}
   ExtraSimWeightMaps {type="TArray<FString>"}
   ExtraSets {type="TArray<FString>"}
   ExtraSimFaceIntMaps {type="TArray<FString>"}
   ExtraRenderWeightMaps {type="TArray<FString>"}
  ]
  outputPins [Collection {type=FManagedArrayCollection}]
  properties {
   ExtraRenderWeightMaps=""
   ExtraSets=""
   ExtraSimFaceIntMaps=""
   ExtraSimWeightMaps=""
   bCopyAllUserAttributesToSimModel=True
   bStripUnusedRenderWeightMaps=True
   bStripUnusedSets=True
   bStripUnusedSimFaceIntMaps=True
   bStripUnusedSimWeightMaps=True
  }
 }
]
