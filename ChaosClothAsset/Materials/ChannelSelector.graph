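# ChannelSelector is a material function in the Chaos Cloth Asset plugin that reduces a four-component value to a scalar using a channel vector.
# It normalizes the Channel input and takes its dot product with the Value input, so a positive one-hot vector returns that single component.
# A Channel with several non-zero components yields a weighted sum scaled by the vector's length, not an average.
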MaterialFunction /ChaosClothAsset/Materials/ChannelSelector
graph {nodes [
  n1 = FunctionInput() {inputName=Channel, inputType=FunctionInput_Vector4, previewValue=<0 0 1 0>}
  FunctionOutput(None=DotProduct(A=Normalize(VectorInput=@n1), B=FunctionInput() {inputName=Value, inputType=FunctionInput_Vector4, previewValue=<1 1 1 1>}))
 ]}
